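# Stereo one-shot MetaSound that plays the completion sound for the FloorMist interaction, layering button, siren and bark waves.
# On play, the button and siren start together, while a trigger delay starts the barks slightly later.
# A three-input stereo mixer sums the layers with extra gain on the button, and a trigger accumulator signals finished once all three end.
# The source uses FloorMist attenuation and concurrency assets under the Fort_UI sound class.
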MetaSoundSource /WildEstate_FloorMist/Assets/Audio/SFX/MSS_FloorMist_Interact_Complete
inputs [
 Barks {type=WaveAsset}
 Button {type=WaveAsset}
 Siren {type=WaveAsset}
 UE.Source.OnPlay {type=Trigger}
]
interfaces [UE.OutputFormat.Stereo:1.0, UE.Source.OneShot:1.0, UE.Source:1.1]
outputs [
 "UE.OutputFormat.Stereo.Audio:0" {type=Audio}
 "UE.OutputFormat.Stereo.Audio:1" {type=Audio}
 UE.Source.OneShot.OnFinished {type=Trigger}
]
pages [Default {
  input_defaults {
   Barks=/WildEstate_FloorMist/Assets/Audio/SFX/wav/FloorMist_Interact_Complete_Barks
   Button=/WildEstate_FloorMist/Assets/Audio/SFX/wav/FloorMist_Interact_Complete_Button
   Siren=/WildEstate_FloorMist/Assets/Audio/SFX/wav/FloorMist_Interact_Complete_Siren
   UE.Source.OnPlay=null
  }
  nodes [
   WavePlayer_0 = UE.Wave Player:Stereo(Wave Asset=^Siren, Play=^UE.Source.OnPlay) {also {version [1,1]}}
   WavePlayer_1 = UE.Wave Player:Stereo(Wave Asset=^Button, Play=^UE.Source.OnPlay) {also {version [1,1]}}
   TriggerDelay_0 = UE.Trigger Delay(In=^UE.Source.OnPlay) {also {defaults {Delay Time=1.3}}}
   WavePlayer_2 = UE.Wave Player:Stereo(Wave Asset=^Barks, Play=@TriggerDelay_0.Out) {also {version [1,1]}}
   AudioMixerAudioMixerStereo3_0 = "AudioMixer.Audio Mixer (Stereo, 3)"(
    In 1 L=@WavePlayer_0.Out Left
    In 1 R=@WavePlayer_0.Out Right
    In 0 L=@WavePlayer_1.Out Left
    In 0 R=@WavePlayer_1.Out Right
    In 2 L=@WavePlayer_2.Out Left
    In 2 R=@WavePlayer_2.Out Right
   ) {also {defaults {Gain 0=1.5}}}
   TriggerAccumulatorTriggerAccumulate3_0 = "TriggerAccumulator.Trigger Accumulate (3)"(In 1=@WavePlayer_0.On Finished, In 0=@WavePlayer_1.On Finished, In 2=@WavePlayer_2.On Finished)
  ]
  links [
   {dst="@UE.OutputFormat.Stereo.Audio:0", src=AudioMixerAudioMixerStereo3_0.Out L}
   {dst="@UE.OutputFormat.Stereo.Audio:1", src=AudioMixerAudioMixerStereo3_0.Out R}
   {dst="@UE.Source.OneShot.OnFinished", src=TriggerAccumulatorTriggerAccumulate3_0.Out}
  ]
 }]
properties {
 assetImportData=&self:AssetImportData
 attenuationSettings=/WildEstate_FloorMist/Assets/Audio/ATT_FloorMist_Interact
 maxDistance=5200
 numChannels=2
 outputFormat=Stereo
 sampleRate=48000
 soundClassObject=&Fort_UI
 concurrencyOverrides {bVolumeScaleCanRelease=true, maxCount=3, resolutionRule=StopOldest, retriggerTime=0.2, volumeScaleReleaseTime=0.1}
 concurrencySet [/WildEstate_FloorMist/Assets/Audio/CONC_FloorMist_Interact]
}
