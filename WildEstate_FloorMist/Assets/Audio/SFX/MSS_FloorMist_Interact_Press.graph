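# One-shot stereo MetaSound for the interact press sound of the Wild Estate FloorMist.
# On play, Random Get nodes pick from the Press and Hold wave arrays, and both waves play at once into a stereo mixer with the Hold layer at half gain.
# A Trigger Accumulate fires OnFinished only after both players finish.
# It uses item-interact attenuation, the Fort_UI sound class and a FloorMist interact concurrency set.
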MetaSoundSource /WildEstate_FloorMist/Assets/Audio/SFX/MSS_FloorMist_Interact_Press
inputs [
 HOld {type=WaveAsset:Array}
 Press {type=WaveAsset:Array}
 UE.Source.OnPlay {type=Trigger}
]
interfaces [UE.OutputFormat.Stereo:1.0, UE.Source.OneShot:1.0, UE.Source:1.1]
outputs [
 "UE.OutputFormat.Stereo.Audio:0" {type=Audio}
 "UE.OutputFormat.Stereo.Audio:1" {type=Audio}
 UE.Source.OneShot.OnFinished {type=Trigger}
]
pages [Default {
  input_defaults {
   UE.Source.OnPlay=null
   HOld [/WildEstate_FloorMist/Assets/Audio/SFX/wav/FloorMist_Interact_Hold]
   Press [/WildEstate_FloorMist/Assets/Audio/SFX/wav/FloorMist_Interact_Press]
  }
  nodes [
   ArrayRandomGet_1 = Array.Random Get:WaveAsset:Array(In Array=^HOld, Next=^UE.Source.OnPlay) {also {defaults {Weights []}}}
   WavePlayer_0 = UE.Wave Player:Stereo(Play=@ArrayRandomGet_1.On Next, Wave Asset=@ArrayRandomGet_1.Value) {also {version [1,1]}}
   ArrayRandomGet_0 = Array.Random Get:WaveAsset:Array(In Array=^Press, Next=^UE.Source.OnPlay)
   WavePlayer_1 = UE.Wave Player:Stereo(Play=@ArrayRandomGet_0.On Next, Wave Asset=@ArrayRandomGet_0.Value) {also {version [1,1]}}
   AudioMixerAudioMixerStereo2_0 = "AudioMixer.Audio Mixer (Stereo, 2)"(In 1 L=@WavePlayer_0.Out Left, In 1 R=@WavePlayer_0.Out Right, In 0 L=@WavePlayer_1.Out Left, In 0 R=@WavePlayer_1.Out Right) {also {defaults {Gain 1=0.5}}}
   TriggerAccumulatorTriggerAccumulate2_0 = "TriggerAccumulator.Trigger Accumulate (2)"(In 1=@WavePlayer_0.On Finished, In 0=@WavePlayer_1.On Finished)
  ]
  links [
   {dst="@UE.OutputFormat.Stereo.Audio:0", src=AudioMixerAudioMixerStereo2_0.Out L}
   {dst="@UE.OutputFormat.Stereo.Audio:1", src=AudioMixerAudioMixerStereo2_0.Out R}
   {dst="@UE.Source.OneShot.OnFinished", src=TriggerAccumulatorTriggerAccumulate2_0.Out}
  ]
 }]
properties {
 assetImportData=&self:AssetImportData
 attenuationSettings=/Game/Sounds/Attenuation/Fort_Item_Interact_Att
 maxDistance=1850
 numChannels=2
 outputFormat=Stereo
 sampleRate=48000
 soundClassObject=&Fort_UI
 concurrencyOverrides {bVolumeScaleCanRelease=true, maxCount=3, resolutionRule=StopOldest, retriggerTime=0.2, volumeScaleReleaseTime=0.1}
 concurrencySet [/WildEstate_FloorMist/Assets/Audio/CONC_FloorMist_Interact]
}
