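# Sound cue for a looping spot ambience hum in the FloorMist area, apparently from a switched-on light or alarm.
# A Looping node repeats a single wave player using the FloorMist LightOn Hum wave.
# It uses a shared light-spot attenuation asset and concurrency set, and routes to the outdoor ambient sound class.
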SoundCue /WildEstate_FloorMist/Assets/Audio/SFX/Spot_AMB_FloorMist_AlarmHum
root=Looping_1
nodes [
 {id=Looping_1, children [WavePlayer_2]}
 {id=WavePlayer_2, properties {soundWaveAssetPtr=/WildEstate_FloorMist/Assets/Audio/SFX/wav/FloorMist_LightOn_Hum}}
]
properties {
 attenuationSettings=/WildEstate_FloorMist/Assets/Audio/ATT_FloorMist_Light_Spot_AMB
 maxDistance=1200
 soundClassObject=&Fort_Ambient_Outdoors_Athena
 volumeMultiplier=1
 concurrencySet [/WildEstate_FloorMist/Assets/Audio/CONC_FloorMist_Light_Spot_AMB]
 preEffectBusSends [{}]
}
